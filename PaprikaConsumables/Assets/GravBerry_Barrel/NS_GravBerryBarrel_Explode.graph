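# Niagara explosion effect for the GravBerry barrel consumable, bursting it into debris and blue-purple liquid.
# Debris meshes use spawn groups to vary scale and visibility, then tumble and bounce with collision.
# An expanding sphere with colour-animated component renderer, mesh spikes, flare, foam, smoke, droplets and curl-noise specks layer the splash.
# Some emitters distance-cull or simulate on GPU; the flare scales by time of day and front-end context.
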NiagaraSystem /PaprikaConsumables/Assets/GravBerry_Barrel/NS_GravBerryBarrel_Explode
dependencies=null
emitters {
 BarrelDebris {
  emitterUpdateScript [
   &EmitterState()
   &SpawnBurst_Instantaneous(Spawn Count=1, Use Spawn Probability=false)
   &SpawnBurst_Instantaneous(Spawn Count=4, Spawn Group=1, Use Spawn Probability=false) {module=SpawnBurst_Instantaneous001}
  ]
  particleSpawnScript [
   /Niagara/Modules/Spawn/Initialization/InitializeParticle(
    Lifetime=1.65
    Mesh Scale=&MaskVectorBySpawnGroup(Masked Vector=<2.75 2.75 2.75>, Passthrough Vector=<1.25 1.25 1.25>)
    Module.Write Color=true
    Module.Write Lifetime=true
    Module.Write Mass=true
    Module.Write Material Random=true
    Module.Write Position=false
    Module.Write Scale=true
    Module.Write SpriteRotation=false
    Module.Write SpriteSize=false
    Position=&SimulationPosition()
   ) {enabled=false}
   /Niagara/Modules/Spawn/Initialization/V2/InitializeParticle(
    Lifetime=1.65
    Mass Mode=::NewEnumerator1
    Mesh Scale=&MaskVectorBySpawnGroup(Masked Vector=<2.75 2.75 2.75>, Passthrough Vector=<1.25 1.25 1.25>)
    Mesh Scale Mode=::NewEnumerator4
    UsePositionOffset=false
    Write Material Random=true
   ) {module=InitializeParticle001}
   /Niagara/Modules/Spawn/Orientation/V2/InitialMeshOrientation(
    Orientation Vector=/Niagara/DynamicInputs/Random/RandomVector(
     Vector Scale:NiagaraFloat=1
     Constraint Vector:Vector3f=<0 0 1>
     Randomness Mode:ENiagaraRandomnessMode=::NewEnumerator0
     Random Seed:NiagaraInt32=0
     Fixed Random Seed:NiagaraBool=false
    )
    Use Orientation Vector=true
    Use Rotation Vector=false
   )
   &MeshRotationForce(
    "Lever Radius (cm)":NiagaraFloat=&UniformRangedFloat(Minimum=200, Maximum=400)
    Rotation=&RandomVector(
     Vector Scale:NiagaraFloat=1
     Constraint Vector:Vector3f=<0 0 1>
     Randomness Mode:ENiagaraRandomnessMode=::NewEnumerator0
     Random Seed:NiagaraInt32=0
     Fixed Random Seed:NiagaraBool=false
    )
   )
   &AddVelocityInCone(
    Cone Angle=60
    Cone Axis=<0 0 1>
    Module.Use Velocity Falloff On Cone Axis=false
    Velocity Strength=&UniformRangedFloat(Minimum=300, Maximum=1150)
   )
   &MeshRotationForce(
    Coordinate Space:ENiagaraMeshTransforms=::NewEnumerator3
    Rotation=&UniformRangedVector(Minimum=<-10 -10 -10>, Maximum=<10 10 10>)
   ) {module=MeshRotationForce001}
   &ApplyInitialForces()
   /SetVariables(Particles.RandomGravity:NiagaraFloat=&UniformRangedFloat(Minimum=-1200, Maximum=-2000)) {isSetParameters=true}
   /SetVariables(Particles.VisibilityTag=&SpawnGroup()) {isSetParameters=true}
  ]
  particleUpdateScript [
   &ParticleState()
   &Drag(Use Linear Drag=true, Use Rotational Drag=false)
   &AccelerationForce(Acceleration=&MakeVector(X=0, Y=0, Z=$Particles.RandomGravity))
   &Collision(
    Age Colliding Particles=false
    Control Roll On Collision=true
    Enable Rest State=false
    Module.EnableMaxCollisionCount=false
    Module.Randomize Collision Normal Vector=true
    Particle Radius Scale=12
    Radius Calculation Type=::NewEnumerator3
    Restitution=0.5
    ScaleRateOfRotation=false
   )
   &SolveForcesAndVelocity(Clamp Velocity=false, Limit Acceleration=false)
   &SolveRotationalForcesAndVelocity()
   &Color(Color=<rgba 0.85 0.7 1 1>) {enabled=false}
   &ScaleMeshSize(Scale Factor=&VectorFromFloat(Value=&FloatFromCurve(FloatCurve=DataInterface() {propertyValues {Curve {keys [<at 0.858 1>,<at 1 0>]}}})))
  ]
  rendererClasses [&NiagaraMeshRendererProperties]
  renderers [
   {rendererClass=&NiagaraMeshRendererProperties, rendererIndex=0}
   {rendererClass=&NiagaraMeshRendererProperties, rendererIndex=1}
  ]
 }
 Flare {
  emitterUpdateScript [
   &EmitterState(Life Cycle Mode=::NewEnumerator1, Loop Behavior=::NewEnumerator1, Loop Delay=0.1, Loop Duration=1, UseLoopDelay=false)
   &SpawnBurst_Instantaneous(Spawn Count=1, Use Spawn Probability=false)
  ]
  particleSpawnScript [
   &InitializeParticle(
    Color=<rgba 0 0.3333 1 1>
    Lifetime=0.27
    Position Offset=<0 0 80>
    Sprite Size=<1200 400>
    Sprite Size Mode=::NewEnumerator4
    UsePositionOffset=true
    Write Material Random=true
   )
   &DynamicMaterialParameters(
    Index 0 Param 1=&RandomRangeFloat(Minimum=0.4, Maximum=0.6)
    Index 0 Param 2=1
    Index 0 Param 3=1
    Index 0 Param 4=1
    Module.Param0WriteEnabled=true
    Module.Param1WriteEnabled=true
    Module.Param2WriteEnabled=true
    Module.Param3WriteEnabled=true
   )
   &CameraOffset(Camera Offset Amount=70)
  ]
  particleUpdateScript [
   /SetVariables(Particles.Dark_Glow=&MakeLinearColorFromVectorAndFloat(
     "Vector (RGB)"=&MakeVectorFromLinearColorRGB(Color=&MultiplyLinearColor(
       Linear Color=<rgba 0.4167 0.2847 0 0.5>
       Scale Factor=&MakeVector4_FromFloat(Float=&FNTimeOfDayScalar(Night Scale=0.7, Morning Scale=0.3, Evening Scale=0.4))
      ))
     "Float (Alpha)"=&Multiply_Float(
      A=&Multiply_Float(A=1, B=0.7)
      B=&FloatFromCurve(FloatCurve=DataInterface() {propertyValues {Curve {keys [<at 0 0>,<at 0.2 1>,<at 0.7 0.6>,<at 1 0>]}}})
     )
    )) {isSetParameters=true}
   /SetVariables(Particles.Dark_Glow_Size=&Multiply_Vector2(
     A=&Vector2DFromFloat(Value=200)
     B=&Vector2DFromFloat(Value=&FloatFromCurve(
       FloatCurve=DataInterface() {propertyValues {Curve {keys [<at 0 1.5 RCIM_Cubic>,<at 1 0.5>]}}}
       Scale Curve=&FN_FrontEndFloatModifier(Gameplay=1, Front End=0.6)
      ))
    )) {isSetParameters=true}
   &ParticleState()
   &ScaleColor(
    Scale Alpha=&Multiply_Float(
     A=&FNTimeOfDayScalar(Night Scale=0.7, Morning Scale=0.3, Evening Scale=0.4)
     B=&FloatFromCurve(FloatCurve=DataInterface() {propertyValues {Curve {keys [
          {interpMode=RCIM_Cubic, tangentMode=RCTM_User, time=0, value=0}
          {interpMode=RCIM_Cubic, tangentMode=RCTM_User, time=0.15, value=1}
          {interpMode=RCIM_Cubic, tangentMode=RCTM_User, time=0.7, value=1}
          {interpMode=RCIM_Cubic, tangentMode=RCTM_User, time=1, value=0}
         ]}}})
    )
    Scale RGB=<7 7 7>
    ScaleA=true
    ScaleRGB=true
   )
   &ScaleSpriteSize(Scale Factor=&Vector2DFromFloat(Value=&FloatFromCurve(
      FloatCurve=DataInterface() {propertyValues {Curve {keys [<at 0 1.5 RCIM_Cubic>,<at 1 0>]}}}
      Scale Curve=&FN_FrontEndFloatModifier(Gameplay=1, Front End=0.6)
     )))
  ]
  rendererClasses [&NiagaraSpriteRendererProperties]
  renderers [
   {rendererClass=&NiagaraSpriteRendererProperties}
   {rendererClass=&NiagaraSpriteRendererProperties, rendererIndex=1}
  ]
 }
 Foam {
  emitterUpdateScript [
   &EmitterState()
   &SpawnBurst_Instantaneous(Spawn Count=5, Use Spawn Probability=false, UseLoopCountLimit=false)
   &SpawnRate(SpawnRate=90, Use Spawn Probability=false) {enabled=false}
  ]
  particleSpawnScript [
   &InitializeParticle(
    Lifetime Max=0.9
    Lifetime Min=0.55
    Lifetime Mode=::NewEnumerator1
    Sprite Size Mode=::NewEnumerator1
    Sprite UV Mode=::NewEnumerator1
    Uniform Sprite Size Max=120
    Uniform Sprite Size Min=90
    UsePositionOffset=false
    Write Material Random=true
   )
   &SubUVAnimation(End Frame=3, SubUV Animation Mode:ENiagara_SubUVLookupMode=::NewEnumerator2)
   &SystemLocation(MaskSpawnGroup=false, Offset=<0 0 40>)
   &ShapeLocation(
    MaskSpawnGroup=false
    Sphere Distribution=::NewEnumerator1
    Sphere Radius=20
    U Position=&Multiply_Float(A=$Emitter.Age, B=5)
    UseOwnerScale=true
    V Position=&RandomRangeFloat(Minimum=0.98, Maximum=1)
   )
   &AddVelocity(
    Cone Axis=<0 0 1>
    Module.UseConeFalloff=false
    Module.UseVelDistribution=false
    Override Seed=false
    Velocity Mode=::NewEnumerator2
    Velocity Speed=&RandomRangeFloat(Minimum=200, Maximum=500)
   ) {module=AddVelocity001}
   &CameraOffset(Camera Offset Amount=35)
  ]
  particleUpdateScript [
   &ParticleState()
   &GravityForce()
   &Drag(
    Drag=&RandomRangeFloat(Minimum=0.5, Maximum=1)
    Rotational Drag=50
    Use Linear Drag=true
    Use Rotational Drag=false
   )
   &InheritSourceMovement(
    Applied Position Scale=&VectorFromFloat(Value=$System.Inherit Source Movement)
    ApplyForce=false
    ApplyPosition=true
    ApplyPositionFalloff=false
   )
   &SolveForcesAndVelocity(Clamp Velocity=false, Limit Acceleration=false)
   &SpriteRotationRate(Rotation Rate=&FloatFromCurve(
     FloatCurve=DataInterface() {propertyValues {Curve {keys [<at 0 1>,<at 1 0>]}}}
     Scale Curve=&MakeCustomFloatFromBool(Bool=&RandomBool(), True Float=-90, False Float=90)
    )) {enabled=false}
   &ScaleSpriteSize(Scale Factor=&Vector2DFromFloat(Value=&FloatFromCurve(
      FloatCurve=DataInterface() {propertyValues {Curve {keys [
          {arriveTangent=0.6251, interpMode=RCIM_Cubic, leaveTangent=0.6251, tangentMode=RCTM_User, time=0, value=0.1}
          {interpMode=RCIM_Cubic, leaveTangent=0.4735, time=1, value=1}
         ]}}}
      Scale Curve=3
     )))
   &ScaleColor(
    Scale Alpha=&FloatFromCurve(FloatCurve=DataInterface() {propertyValues {Curve {keys [{interpMode=RCIM_Cubic, tangentMode=RCTM_User, time=0.3, value=0.76}, {interpMode=RCIM_Cubic, tangentMode=RCTM_User, time=1, value=0}]}}})
    Scale RGB=&Lerp_Vector(
     A=<0 1 3>
     B=<0.25 0.07 1>
     Alpha=&FloatFromCurve(FloatCurve=DataInterface() {propertyValues {Curve {keys [{interpMode=RCIM_Cubic, tangentMode=RCTM_User, time=0.16, value=0}, {interpMode=RCIM_Cubic, tangentMode=RCTM_User, time=0.6, value=1}]}}})
    )
    ScaleA=true
    ScaleRGB=true
   )
   &DynamicMaterialParameters(
    Index 0 Param 2=&FloatFromCurve(FloatCurve=DataInterface() {propertyValues {Curve {keys [{arriveTangent=-1.9579, interpMode=RCIM_Cubic, time=0.1, value=1}, {interpMode=RCIM_Cubic, tangentMode=RCTM_User, time=0.3, value=0}]}}})
    Index 0 Param 4=&FloatFromCurve(FloatCurve=DataInterface() {propertyValues {Curve {keys [{arriveTangent=1.1512, interpMode=RCIM_Cubic, leaveTangent=1.1512, tangentMode=RCTM_User, time=0, value=0}, <at 1 1 RCIM_Cubic>]}}})
    Module.Param0WriteEnabled=false
    Module.Param1WriteEnabled=true
    Module.Param2WriteEnabled=false
    Module.Param3WriteEnabled=true
   )
  ]
  rendererClasses [&NiagaraSpriteRendererProperties]
  renderers [
   {rendererClass=&NiagaraSpriteRendererProperties}
   {rendererClass=&NiagaraSpriteRendererProperties, rendererIndex=1}
  ]
 }
 GeoSpikes {
  emitterUpdateScript [
   &EmitterState(Enable Distance Culling=true, MaxDistance=8500, Module.bUseMaxDistance=true, Module.bUseMinDistance=false, Scalability Mode=::NewEnumerator1)
   &SpawnBurst_Instantaneous(Spawn Count=4, Use Spawn Probability=false)
  ]
  particleSpawnScript [
   &InitializeParticle(
    Color Maximum=<rgba 2.08 0.16 4 1>
    Color Minimum=<rgba 0 1.6667 4 1>
    Color Mode=::NewEnumerator2
    Lifetime=0.16
    Mesh Scale Mode=::NewEnumerator1
    Mesh Uniform Scale Max=3
    Mesh Uniform Scale Min=2
    UsePositionOffset=false
    Write Material Random=true
   ) {module=InitializeParticle001}
   /Niagara/Modules/Spawn/Orientation/V3/InitialMeshOrientation(
    Mesh Orientation Mode=::NewEnumerator2
    Orientation Vector=<0 0 1>
    Rotation=&UniformRangedVector(Minimum=<-0.125 -0.125 0>, Maximum=<0.125 0.125 0>)
    Rotation Coordinate Space=::NewEnumerator2
    Use Rotation Vector=true
   ) {module=InitialMeshOrientation001}
   &SystemLocation(MaskSpawnGroup=false, Offset=<0 0 100>)
  ]
  particleUpdateScript [
   &ParticleState()
   &CameraOffset(Camera Offset Amount=20)
   &ScaleMeshSize(Scale Factor=&VectorFromFloat(Value=&FloatFromCurve(
      FloatCurve=DataInterface() {propertyValues {Curve {keys [{interpMode=RCIM_Cubic, leaveTangent=1.3964, tangentMode=RCTM_Break, time=0, value=0}, <at 1.5 1>]}}}
      Scale Curve=1.25
     )))
   &ScaleColor(
    Scale Alpha=&FloatFromCurve(FloatCurve=DataInterface() {propertyValues {Curve {keys [<at 0.85 1>,<at 1 0>]}}})
    Scale RGB=&VectorFromFloat(Value=&FloatFromCurve(FloatCurve=DataInterface() {propertyValues {Curve {keys [{interpMode=RCIM_Cubic, tangentMode=RCTM_Break, time=0, value=5}, {interpMode=RCIM_Cubic, tangentMode=RCTM_Break, time=0.7, value=0.6}]}}}))
    ScaleA=true
    ScaleRGB=true
   )
   &ScaleMeshSize(Scale Factor=&VectorFromCurve(VectorCurve=DataInterface() {propertyValues {
       XCurve {keys [
         {interpMode=RCIM_Cubic, leaveTangent=6.7773, tangentMode=RCTM_Break, time=0, value=0}
         {interpMode=RCIM_Cubic, tangentMode=RCTM_Break, time=0.7, value=2}
        ]}
       YCurve {keys [<at 0 1 RCIM_Cubic>, {arriveTangent=-1.0152, interpMode=RCIM_Cubic, leaveTangent=-1.0152, tangentMode=RCTM_User, time=0.8, value=0.2}]}
       ZCurve {keys [<at 0 1 RCIM_Cubic>, {arriveTangent=-1.0152, interpMode=RCIM_Cubic, leaveTangent=-1.0152, tangentMode=RCTM_User, time=0.8, value=0.2}]}
      }})) {module=ScaleMeshSize001}
  ]
  rendererClasses [&NiagaraMeshRendererProperties]
  renderers [{rendererClass=&NiagaraMeshRendererProperties, rendererIndex=0}]
 }
 SlurpWispySmoke {
  emitterUpdateScript [
   &EmitterState(Life Cycle Mode=::NewEnumerator1, Loop Behavior=::NewEnumerator1, Loop Duration=1, UseLoopDelay=false)
   &SpawnBurst_Instantaneous(Spawn Count=&UniformRangedInt(Minimum=10, Maximum=6), Use Spawn Probability=false)
  ]
  particleSpawnScript [
   /Niagara/Modules/Spawn/Initialization/InitializeParticle(
    Color=&UniformRangedLinearColor(Minimum=<rgba 1 1 1 0.7>, Maximum=<rgba 0.25 0.25 0.25 0.7>)
    Lifetime=&UniformRangedFloat(Minimum=1.4, Maximum=1.8)
    Module.Write Color=true
    Module.Write Lifetime=true
    Module.Write Mass=true
    Module.Write Material Random=true
    Module.Write Position=true
    Module.Write Scale=false
    Module.Write SpriteRotation=true
    Module.Write SpriteSize=true
    Position=&Add_Vector(A=Linked() {linkedVariable {name=Engine.Owner.Position, type=NiagaraPosition}}, B=<0 0 50>)
    Sprite Rotation=&UniformRangedFloat(Minimum=0, Maximum=360)
    Sprite Size=&Vector2DFromFloat(Value=&UniformRangedFloat(Minimum=100, Maximum=265))
   ) {enabled=false}
   /Niagara/Modules/Spawn/Initialization/V2/InitializeParticle(
    Color=&UniformRangedLinearColor(Minimum=<rgba 1 1 1 0.7>, Maximum=<rgba 0.25 0.25 0.25 0.7>)
    Lifetime Max=1.4
    Lifetime Mode=::NewEnumerator1
    Sprite Rotation Mode=::NewEnumerator2
    Sprite Size Mode=::NewEnumerator1
    Uniform Sprite Size Max=265
    Uniform Sprite Size Min=100
    UsePositionOffset=false
    Write Material Random=true
   ) {module=InitializeParticle001}
   /SetVariables(Particles.Position=&AddVectorToPosition(Position=$Engine.Emitter.SimulationPosition, Vector=<0 0 50>)) {isSetParameters=true}
   &AddVelocityInCone(
    Cone Angle=120
    Cone Axis=<0 0 1>
    Module.Use Velocity Falloff On Cone Axis=false
    Velocity Strength=&UniformRangedFloat(Minimum=500, Maximum=755)
   )
   &SphereLocation(
    MaskSpawnGroup=false
    Override Seed=false
    Sphere Radius=10
    Surface Only=false
    UseNonUniformScale=false
    UseOffset=false
    UseSphereOrientation=false
   )
   &Color(Color=&MultiplyLinearColor(
     Linear Color=$Particles.Color
     Scale Factor=&MakeVector4_FromLinearColor(LinearColor=&UniformRangedLinearColor(Minimum=<rgba 0.03 0.5958 1 0.23>, Maximum=<rgba 0.4575 0.07 1 0.23>))
    ))
   /SetVariables(Particles.UVScale:Vector2f=&UniformRangedVector2D(Minimum=<-1 1>, Maximum=<-1 1>)) {isSetParameters=true}
  ]
  particleUpdateScript [
   &ParticleState()
   &Drag(Drag=&UniformRangedFloat(Minimum=2, Maximum=3), Use Linear Drag=true, Use Rotational Drag=false)
   &GravityForce(Gravity=<0 0 -650>)
   &SolveForcesAndVelocity(Clamp Velocity=false, Limit Acceleration=false)
   &SpriteRotationRate(Rotation Rate=&Multiply_Float(
     A=&Lerp_Float(A=400, B=-400, Alpha=$Particles.Initial.MaterialRandom)
     B=&Lerp_Float(A=1, B=0, Alpha=$Particles.NormalizedAge)
    ))
   &ScaleColor(
    Scale Alpha=&Lerp_Float(A=1, B=0, Alpha=&Power(a=$Particles.NormalizedAge, b=2))
    ScaleA=true
    ScaleRGB=true
   )
   &ScaleSpriteSize(Scale Factor=&Vector2DFromFloat(Value=&Lerp_Float(A=1, B=2, Alpha=$Particles.NormalizedAge)))
   /Niagara/Modules/Update/SubUV/SubUVAnimation(Number Of Frames=63)
   &DynamicMaterialParameters(
    Index 0 Param 1=4
    Index 0 Param 2=0.75
    Index 0 Param 3=1
    Index 0 Param 4=-0.1
    Module.Index0Param0Write=true
    Module.Index0Param1Write=true
    Module.Index0Param2Write=true
    Module.Index0Param3Write=true
   )
   &ScaleColor(
    Scale Alpha=&FloatFromCurve(FloatCurve=DataInterface() {propertyValues {Curve {keys [<at 0 0>,<at 0.2 1>]}}})
    ScaleA=true
    ScaleRGB=true
   ) {module=ScaleColor001}
  ]
  rendererClasses [&NiagaraSpriteRendererProperties]
  renderers [{rendererClass=&NiagaraSpriteRendererProperties}]
 }
 Soda_pop {
  simTarget=GPUComputeSim
  emitterUpdateScript [
   &EmitterState(
    Enable Distance Culling=true
    Life Cycle Mode=::NewEnumerator1
    Loop Behavior=::NewEnumerator1
    Loop Duration=0.55
    MaxDistance=2250
    Module.bUseMaxDistance=true
    Module.bUseMinDistance=false
    Scalability Mode=::NewEnumerator1
    UseLoopDelay=false
   )
   &SpawnRate(SpawnRate=150, Use Spawn Probability=false)
  ]
  particleSpawnScript [
   /Niagara/Modules/Spawn/Initialization/InitializeParticle(
    Color=&UniformRangedLinearColor(Minimum=<rgba 0 1 4 1>, Maximum=<rgba 1.6667 0 4 1>)
    Lifetime=&UniformRangedFloat(Minimum=0.18, Maximum=1)
    Module.Write Color=true
    Module.Write Lifetime=true
    Module.Write Mass=true
    Module.Write Material Random=true
    Module.Write Position=false
    Module.Write Scale=false
    Module.Write SpriteRotation=false
    Module.Write SpriteSize=true
    Position=&SimulationPosition()
    Sprite Size=&Vector2DFromFloat(Value=&UniformRangedFloat(Minimum=4, Maximum=12))
   ) {enabled=false}
   /Niagara/Modules/Spawn/Initialization/V2/InitializeParticle(
    Color Maximum=<rgba 1.6667 0 4 1>
    Color Minimum=<rgba 0 1 4 1>
    Color Mode=::NewEnumerator2
    Lifetime Max=1
    Lifetime Min=0.18
    Lifetime Mode=::NewEnumerator1
    Sprite Size Mode=::NewEnumerator1
    Uniform Sprite Size Max=12
    Uniform Sprite Size Min=4
    UsePositionOffset=false
    Write Material Random=true
   ) {module=InitializeParticle001}
   &SphereLocation(
    MaskSpawnGroup=false
    Non Uniform Scale=<1 1 0.6>
    Offset=<0 0 10>
    Override Seed=false
    Sphere Radius=250
    Surface Only=false
    UseNonUniformScale=true
    UseOffset=true
    UseSphereOrientation=false
   )
   &AddVelocity(Velocity=<0 0 150>)
  ]
  particleUpdateScript [
   &ParticleState()
   /Niagara/Modules/Update/Forces/CurlNoiseForce(Noise Strength=800, Pan Noise=false)
   &SolveForcesAndVelocity(Clamp Velocity=false, Limit Acceleration=false)
   &ScaleSpriteSize(Scale Factor=&Vector2DFromFloat(Value=&FloatFromCurve(FloatCurve=DataInterface() {propertyValues {Curve {keys [{arriveTangent=3.6722, interpMode=RCIM_Cubic, leaveTangent=3.6722, tangentMode=RCTM_User, time=0, value=0}, <at 0.8 1>, <at 1 2>]}}})))
   &ScaleColor(
    Scale Alpha=&FloatFromCurve(FloatCurve=DataInterface() {propertyValues {Curve {keys [
         <at 0 1>
         {arriveTangent=-4.4715, interpMode=RCIM_Cubic, leaveTangent=-4.4715, tangentMode=RCTM_User, time=0.8022, value=0.9932}
         {arriveTangent=-7.8513, interpMode=RCIM_Cubic, leaveTangent=-7.8513, tangentMode=RCTM_User, time=1, value=0}
        ]}}})
    ScaleA=true
    ScaleRGB=true
   )
  ]
  rendererClasses [&NiagaraSpriteRendererProperties]
  renderers [{rendererClass=&NiagaraSpriteRendererProperties}]
 }
 ThrownDrops001 {
  emitterUpdateScript [
   &EmitterState(
    Enable Distance Culling=true
    Life Cycle Mode=::NewEnumerator1
    Loop Behavior=::NewEnumerator1
    Loop Delay=0.02
    Loop Duration=1
    MaxDistance=4000
    Module.bUseMaxDistance=true
    Module.bUseMinDistance=false
    Scalability Mode=::NewEnumerator1
    UseLoopDelay=false
   )
   &SpawnBurst_Instantaneous(Spawn Count=&UniformRangedInt(Minimum=5, Maximum=25), Use Spawn Probability=false)
  ]
  particleSpawnScript [
   /Niagara/Modules/Spawn/Initialization/InitializeParticle(
    Color=&UniformRangedLinearColor(Minimum=<rgba 1 1 1 1>, Maximum=<rgba 1 1 1 1>)
    Lifetime=&UniformRangedFloat(Minimum=0.25, Maximum=1.5)
    Mass=&UniformRangedFloat(Minimum=0.75, Maximum=1.75)
    Module.Write Color=true
    Module.Write Lifetime=true
    Module.Write Mass=true
    Module.Write Material Random=true
    Module.Write Position=false
    Module.Write Scale=false
    Module.Write SpriteRotation=false
    Module.Write SpriteSize=true
    Position=&SimulationPosition()
    Sprite Size=&Vector2DFromFloat(Value=&UniformRangedFloat(Minimum=0.5, Maximum=4))
   ) {enabled=false}
   /Niagara/Modules/Spawn/Initialization/V2/InitializeParticle(
    Lifetime Max=1
    Lifetime Min=0.25
    Lifetime Mode=::NewEnumerator1
    Mass=&UniformRangedFloat(Minimum=0.75, Maximum=1.75)
    Mass Mode=::NewEnumerator1
    Sprite Size Mode=::NewEnumerator3
    Uniform Sprite Size=&MakeFloatFromVector2d(Vector 2D=&Vector2DFromFloat(Value=&UniformRangedFloat(Minimum=0.5, Maximum=4)))
    UsePositionOffset=false
    Write Material Random=true
   ) {module=InitializeParticle001}
   /Niagara/Modules/Update/Mass/CalculateSizeAndRotationalInertiaByMass(Depth=0, Module.Manually Enter Density=false)
   &SphereLocation(
    MaskSpawnGroup=false
    Offset=<0 0 50>
    Override Seed=false
    Sphere Radius=33
    Surface Only=false
    UseNonUniformScale=false
    UseOffset=true
    UseSphereOrientation=false
   )
   &AddVelocityFromPoint(
    Module.Use Falloff Distance=false
    Origin Offset=<0 0 50>
    Velocity Strength=&UniformRangedFloat(Minimum=260, Maximum=500)
   )
   &AddVelocity(Velocity=<0 0 500>)
  ]
  particleUpdateScript [
   &ParticleState()
   &GravityForce(Gravity=<0 0 -1650>)
   &ScaleSpriteSize(Uniform Curve Sprite Scale=DataInterface() {propertyValues {Curve {keys [{interpMode=RCIM_Cubic, tangentMode=RCTM_User, time=0, value=0.3}, {interpMode=RCIM_Cubic, tangentMode=RCTM_User, time=0.2, value=1}]}}})
   &Drag(Drag=&UniformRangedFloat(Minimum=2, Maximum=3), Use Linear Drag=true, Use Rotational Drag=false)
   &ScaleColor(
    Scale Alpha=&FloatFromCurve(FloatCurve=DataInterface() {propertyValues {Curve {keys [{interpMode=RCIM_Cubic, tangentMode=RCTM_User, time=0.1, value=1}, {interpMode=RCIM_Cubic, tangentMode=RCTM_User, time=1, value=0}]}}})
    Scale RGB=&MakeVectorFromLinearColorRGB(Color=&ColorFromCurve(ColorCurve=DataInterface() {propertyValues {
        AlphaCurve {keys []}
        BlueCurve {keys [<at 0.1241 10>,<at 0.4315 1.5>,<at 1 0.8 RCIM_Cubic>]}
        GreenCurve {keys [<at 0.1241 6.25>,<at 0.4315 1.2>,<at 1 0.56 RCIM_Cubic>]}
        RedCurve {keys [<at 0.1241 1>,<at 0.4315 1.25>,<at 1 0.76 RCIM_Cubic>]}
       }}))
    ScaleA=true
    ScaleRGB=true
   )
   &SolveForcesAndVelocity(Clamp Velocity=false, Limit Acceleration=false)
   &ScaleSpriteSizeBySpeed(Max Scale Factor=<2 8>)
  ]
  rendererClasses [&NiagaraSpriteRendererProperties]
  renderers [{rendererClass=&NiagaraSpriteRendererProperties}]
 }
 ThrownDrops002 {
  simTarget=GPUComputeSim
  emitterUpdateScript [
   &EmitterState(
    Enable Distance Culling=true
    Life Cycle Mode=::NewEnumerator1
    Loop Behavior=::NewEnumerator1
    Loop Delay=0.01
    Loop Duration=1
    Module.bUseMaxDistance=true
    Module.bUseMinDistance=false
    Scalability Mode=::NewEnumerator1
    UseLoopDelay=false
   )
   &SpawnBurst_Instantaneous(Spawn Count=&UniformRangedInt(Minimum=30, Maximum=80), Use Spawn Probability=false)
  ]
  particleSpawnScript [
   /Niagara/Modules/Spawn/Initialization/InitializeParticle(
    Lifetime=&UniformRangedFloat(Minimum=0.5, Maximum=0.75)
    Mass=&UniformRangedFloat(Minimum=0.75, Maximum=1.75)
    Module.Write Color=true
    Module.Write Lifetime=true
    Module.Write Mass=true
    Module.Write Material Random=true
    Module.Write Position=false
    Module.Write Scale=false
    Module.Write SpriteRotation=false
    Module.Write SpriteSize=true
    Position=&SimulationPosition()
    Sprite Size=&Vector2DFromFloat(Value=&UniformRangedFloat(Minimum=10, Maximum=20))
   ) {enabled=false}
   /Niagara/Modules/Spawn/Initialization/V2/InitializeParticle(
    Lifetime Max=0.75
    Lifetime Min=0.5
    Lifetime Mode=::NewEnumerator1
    Mass Max=1.75
    Mass Mode=::NewEnumerator3
    Sprite Size Mode=::NewEnumerator1
    Uniform Sprite Size Max=20
    Uniform Sprite Size Min=10
    UsePositionOffset=false
    Write Material Random=true
   ) {module=InitializeParticle001}
   &SphereLocation(
    MaskSpawnGroup=false
    Offset=<0 0 50>
    Override Seed=false
    Sphere Radius=20
    Surface Only=false
    UseNonUniformScale=false
    UseOffset=true
    UseSphereOrientation=false
   )
   &AddVelocityFromPoint(
    Module.Use Falloff Distance=false
    Origin Offset=<0 0 50>
    Velocity Strength=&UniformRangedFloat(Minimum=800, Maximum=1300)
   )
   &AddVelocity(Velocity=<0 0 800>)
  ]
  particleUpdateScript [
   &ParticleState()
   &GravityForce(Gravity=<0 0 -2400>)
   &Drag(Drag=&UniformRangedFloat(Minimum=1, Maximum=3), Use Linear Drag=true, Use Rotational Drag=false)
   &ScaleColor(
    Scale Alpha=&FloatFromCurve(FloatCurve=DataInterface() {propertyValues {Curve {keys [{interpMode=RCIM_Cubic, tangentMode=RCTM_User, time=0, value=1}, {interpMode=RCIM_Cubic, tangentMode=RCTM_User, time=1, value=0}]}}})
    Scale RGB=&MakeVectorFromLinearColorRGB(Color=&ColorFromCurve(ColorCurve=DataInterface() {propertyValues {
        AlphaCurve {keys []}
        BlueCurve {keys [<at 0.0907 10>,<at 0.5296 1.5>,<at 1 0.8 RCIM_Cubic>]}
        GreenCurve {keys [<at 0.0907 6.25>,<at 0.5296 1.2>,<at 1 0.56 RCIM_Cubic>]}
        RedCurve {keys [<at 0.0907 1>,<at 0.5296 1.25>,<at 1 0.76 RCIM_Cubic>]}
       }}))
    ScaleA=true
    ScaleRGB=true
   )
   &SolveForcesAndVelocity(Clamp Velocity=false, Limit Acceleration=false)
   &ScaleSpriteSizeBySpeed(Max Scale Factor=<1 5>, Min Scale Factor=<0 0.05>)
   &ScaleSpriteSize(Uniform Curve Sprite Scale=DataInterface() {propertyValues {Curve {keys [{interpMode=RCIM_Cubic, tangentMode=RCTM_User, time=0, value=0.3}, {interpMode=RCIM_Cubic, tangentMode=RCTM_User, time=0.35, value=1}]}}})
  ]
  rendererClasses [&NiagaraSpriteRendererProperties]
  renderers [{rendererClass=&NiagaraSpriteRendererProperties}]
 }
 initSphere {
  emitterUpdateScript [
   &EmitterState(
    Enable Distance Culling=true
    Life Cycle Mode=::NewEnumerator1
    Loop Behavior=::NewEnumerator1
    Loop Duration=0.1
    MaxDistance=13000
    Module.bUseMaxDistance=true
    Module.bUseMinDistance=false
    Scalability Mode=::NewEnumerator1
    UseLoopDelay=false
   )
   &SpawnBurst_Instantaneous(Spawn Count=1, Use Spawn Probability=false)
  ]
  particleSpawnScript [
   /Niagara/Modules/Spawn/Initialization/InitializeParticle(
    Color=<rgba 0.2575 0.31 0.4 1>
    Lifetime=0.7
    Mesh Scale=&VectorFromFloat(Value=1)
    Module.Write Color=true
    Module.Write Lifetime=true
    Module.Write Mass=true
    Module.Write Material Random=true
    Module.Write Position=true
    Module.Write Scale=true
    Module.Write SpriteRotation=false
    Module.Write SpriteSize=false
    Sprite Size=<5 5>
   )
   &SystemLocation(MaskSpawnGroup=false, Offset=<0 0 50>)
  ]
  particleUpdateScript [
   &ParticleState()
   &ScaleMeshSize(Scale Factor=&VectorFromFloat(Value=&FloatFromCurve(
      FloatCurve=DataInterface() {propertyValues {Curve {keys [
          {interpMode=RCIM_Cubic, leaveTangent=2.1274, tangentMode=RCTM_Break, time=0, value=0}
          {arriveTangent=0.6694, interpMode=RCIM_Cubic, leaveTangent=0.6694, tangentMode=RCTM_User, time=0.494, value=0.654}
          <at 1.5 1>
         ]}}}
      Scale Curve=11
     )))
   &ScaleColor(
    Scale Alpha=0.5
    Scale RGB=&MakeVectorFromLinearColorRGB(Color=&ColorFromCurve(ColorCurve=DataInterface() {propertyValues {
        AlphaCurve {keys []}
        BlueCurve {keys [<at 0.0782 1>, {arriveTangent=-1.0849, interpMode=RCIM_Cubic, leaveTangent=-1.0849, time=0.3467, value=1}, <at 1 0>]}
        GreenCurve {keys [<at 0.0782 0.6833>, {arriveTangent=-0.7413, interpMode=RCIM_Cubic, leaveTangent=-0.7413, time=0.3467, value=0.1}, <at 1 0>]}
        RedCurve {keys [<at 0.0782 0.05>, {arriveTangent=-0.0542, interpMode=RCIM_Cubic, leaveTangent=-0.0542, time=0.3467, value=0.445}, <at 1 0>]}
       }}))
    ScaleA=true
    ScaleRGB=true
   )
   /SetVariables(Particles.LightColor=&ColorFromCurve(ColorCurve=DataInterface() {propertyValues {
       AlphaCurve {keys [<at 0.5925 1 RCIM_Cubic>,<at 1 0 RCIM_Cubic>]}
       BlueCurve {keys [<at -0.0022 0>, {arriveTangent=1.6758, interpMode=RCIM_Cubic, leaveTangent=1.6758, time=0.0394, value=1}, <at 0.5946 1>, <at 1 0 RCIM_Cubic>]}
       GreenCurve {keys [
         <at -0.0022 0>
         {arriveTangent=0.2015, interpMode=RCIM_Cubic, leaveTangent=0.2015, time=0.0394, value=0.7569}
         <at 0.5946 0.1203>
         <at 1 0 RCIM_Cubic>
        ]}
       RedCurve {keys [<at -0.0022 0>,<at 0.0394 0.2292 RCIM_Cubic>,<at 0.5946 0>,<at 1 0 RCIM_Cubic>]}
      }})) {isSetParameters=true}
  ]
  rendererClasses [&NiagaraMeshRendererProperties, &NiagaraComponentRendererProperties]
  renderers [
   {rendererClass=&NiagaraMeshRendererProperties, rendererIndex=0}
   {rendererClass=&NiagaraComponentRendererProperties, rendererIndex=1}
  ]
 }
}
summary {
 systemName=NS_GravBerryBarrel_Explode
 emitters [
  {emitterName=BarrelDebris, rendererClasses [&NiagaraMeshRendererProperties]}
  {emitterName=SlurpWispySmoke, rendererClasses [&NiagaraSpriteRendererProperties]}
  {emitterName=ThrownDrops001, rendererClasses [&NiagaraSpriteRendererProperties]}
  {emitterName=ThrownDrops002, simTarget=GPUComputeSim, rendererClasses [&NiagaraSpriteRendererProperties]}
  {emitterName=initSphere, rendererClasses [&NiagaraMeshRendererProperties, &NiagaraComponentRendererProperties]}
  {emitterName=GeoSpikes, rendererClasses [&NiagaraMeshRendererProperties]}
  {emitterName=Soda_pop, simTarget=GPUComputeSim, rendererClasses [&NiagaraSpriteRendererProperties]}
  {emitterName=FN_PIC_FLIP_Fluid_Explosion, simTarget=GPUComputeSim, rendererClasses [&NiagaraSpriteRendererProperties]}
  {emitterName=Flare, rendererClasses [&NiagaraSpriteRendererProperties]}
  {emitterName=Foam, rendererClasses [&NiagaraSpriteRendererProperties]}
 ]
 userVariables [
  User.bIsFrontEnd {type=NiagaraBool}
  User.bIsFrontEndPreview {type=NiagaraBool}
 ]
}
